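# Niagara spark burst for the LazyLizz emote in blue, gold and purple sprites.
# The Spark emitter fires two bursts as spawn groups, writing the group to VisibilityTag to vary sprite size within one emitter with two renderers.
# Particles launch in a wide upward cone, then spin, slow, fall, flash bright and pop then shrink.
# Disabled Spark_01 and Spark_02 hold the same two bursts as separate emitters.
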NiagaraSystem /BRCosmetics/Animation/Game/MainPlayer/Emotes/LazyLizz/FX/NS_Emote_LazyLizz_Spark
dependencies {
 usedDataInterfaces [&NiagaraDataInterfaceArrayColor, &NiagaraDataInterfaceCurve]
 usedDynamicInputs []
 usedModules [
  &self:SystemSpawnScript.SystemScriptSource.SystemScriptGraph.NiagaraNodeAssignment_0.SetVariables_#1
  &SystemState
 ]
}
emitters {
 Spark {
  emitterUpdateScript [
   &EmitterState()
   &SpawnBurst_Instantaneous(
    Spawn Count=&RandomRangeInt(Minimum=22, Maximum=30)
    Use Spawn Probability=false
    UseLoopCountLimit=false
   )
   &SpawnBurst_Instantaneous(
    Spawn Count=&RandomRangeInt(Minimum=10, Maximum=15)
    Spawn Group=1
    Use Spawn Probability=false
    UseLoopCountLimit=false
   ) {module=SpawnBurst_Instantaneous001}
  ]
  particleSpawnScript [
   &InitializeParticle(
    Color=&SelectLinearColorFromArray(
     Array Sampling Mode=::NewEnumerator0
     Direct Array Index=?
     Color Selection Array=DataInterface() {propertyValues {ColorData [<rgba 0 0.1063 1 1>, <rgba 1 0.7667 0 1>, <rgba 0.6295 0 1 1>]}}
     Random Seed=0
    )
    Lifetime Max=0.8
    Lifetime Min=0.4
    Lifetime Mode=::NewEnumerator1
    Sprite Rotation Mode=::NewEnumerator2
    UsePositionOffset=false
    Write Material Random=true
   )
   /SetVariables(Particles.VisibilityTag=&SpawnGroup()) {isSetParameters=true}
   /SetVariables(Particles.SpriteSize=&Vector2DFromFloat(Value=&MaskFloatBySpawnGroup(
      Masked Float=&RandomRangeFloat(Minimum=14, Maximum=20)
      Passthrough Float=&RandomRangeFloat(Minimum=6.6, Maximum=15.95)
     ))) {isSetParameters=true}
   /SetVariables(Particles.Color_2:LinearColor=&SelectLinearColorFromArray(
     Array Sampling Mode=::NewEnumerator0
     Direct Array Index=?
     Color Selection Array=DataInterface() {propertyValues {ColorData [<rgba 0.0002 0.2061 1 0.1>, <rgba 1 0.7108 0 0.06>, <rgba 0.576 0.0003 1 0.1>]}}
     Random Seed=0
    )) {isSetParameters=true}
   &ShapeLocation(MaskSpawnGroup=false, Override Seed=false, Sphere Radius=5, UseOwnerScale=true)
   &AddVelocity(
    Cone Angle=150
    Cone Axis=<0 0 1>
    Module.UseConeFalloff=false
    Module.UseVelDistribution=false
    Override Seed=false
    Velocity Mode=::NewEnumerator2
    Velocity Speed=&MaskFloat(
     Mask Source=0
     Mask Parameter=?
     Mask=$Particles.VisibilityTag
     Masked Float=&RandomRangeFloat(Minimum=90, Maximum=120)
     Passthrough Float=&MaskFloat(
      Mask Source=0
      Mask Parameter=?
      Mask=$Particles.VisibilityTag
      Masked Float=&RandomRangeFloat(Minimum=90, Maximum=120)
      Passthrough Float=1
     )
    )
   )
  ]
  particleUpdateScript [
   &ParticleState()
   &ScaleSpriteSize(Uniform Curve Sprite Scale=DataInterface() {propertyValues {Curve {keys [
        {interpMode=RCIM_Cubic, leaveTangent=15.2837, tangentMode=RCTM_User, time=0, value=0}
        {interpMode=RCIM_Cubic, tangentMode=RCTM_Break, time=0.1, value=1}
        {interpMode=RCIM_Cubic, tangentMode=RCTM_Break, time=1, value=0}
       ]}}})
   &SpriteRotationRate(Rotation Rate=&RandomRangeFloat(Minimum=-200, Maximum=200))
   &Drag(Drag=&RandomRangeFloat(Minimum=1.4, Maximum=1.8), Use Linear Drag=true, Use Rotational Drag=true)
   &GravityForce(Gravity=&RandomRangeVector(Minimum=<0 0 -90>, Maximum=<0 0 -140>))
   &ScaleColor(
    Scale RGB=&VectorFromFloat(Value=&FloatFromCurve(FloatCurve=DataInterface() {propertyValues {Curve {keys [
          {interpMode=RCIM_Cubic, tangentMode=RCTM_Break, time=0.0952, value=3}
          {arriveTangent=-2.2104, interpMode=RCIM_Cubic, leaveTangent=-2.2104, time=0.2284, value=1.652}
          {interpMode=RCIM_Cubic, tangentMode=RCTM_Break, time=1, value=1}
         ]}}}))
    ScaleA=true
    ScaleRGB=true
   )
   &SolveForcesAndVelocity(Clamp Velocity=false, Limit Acceleration=false)
  ]
  rendererClasses [&NiagaraSpriteRendererProperties]
  renderers [
   {rendererClass=&NiagaraSpriteRendererProperties}
   {rendererClass=&NiagaraSpriteRendererProperties, rendererIndex=1}
  ]
 }
 Spark_01 {
  bEnabled=false
  emitterUpdateScript [
   &EmitterState()
   &SpawnBurst_Instantaneous(
    Spawn Count=&RandomRangeInt(Minimum=22, Maximum=30)
    Use Spawn Probability=false
    UseLoopCountLimit=false
   )
  ]
  particleSpawnScript [
   &InitializeParticle(
    Color=&SelectLinearColorFromArray(
     Array Sampling Mode=::NewEnumerator0
     Direct Array Index=?
     Color Selection Array=DataInterface() {propertyValues {ColorData [<rgba 0 0.1063 1 1>, <rgba 1 0.7667 0 1>, <rgba 0.6295 0 1 1>]}}
     Random Seed=0
    )
    Lifetime Max=0.8
    Lifetime Min=0.4
    Lifetime Mode=::NewEnumerator1
    Sprite Rotation Mode=::NewEnumerator2
    Sprite Size Mode=::NewEnumerator1
    Uniform Sprite Size Max=20
    Uniform Sprite Size Min=14
    UsePositionOffset=false
    Write Material Random=true
   )
   &ShapeLocation(MaskSpawnGroup=false, Override Seed=false, Sphere Radius=5, UseOwnerScale=true)
   &AddVelocity(
    Cone Angle=150
    Cone Axis=<0 0 1>
    Module.UseConeFalloff=false
    Module.UseVelDistribution=false
    Override Seed=false
    Velocity Mode=::NewEnumerator2
    Velocity Speed=&RandomRangeFloat(Minimum=90, Maximum=120)
   )
  ]
  particleUpdateScript [
   &ParticleState()
   &ScaleSpriteSize(Uniform Curve Sprite Scale=DataInterface() {propertyValues {Curve {keys [
        {interpMode=RCIM_Cubic, leaveTangent=15.2837, tangentMode=RCTM_User, time=0, value=0}
        {interpMode=RCIM_Cubic, tangentMode=RCTM_Break, time=0.1, value=1}
        {interpMode=RCIM_Cubic, tangentMode=RCTM_Break, time=1, value=0}
       ]}}})
   &SpriteRotationRate(Rotation Rate=&RandomRangeFloat(Minimum=-200, Maximum=200))
   &Drag(Drag=&RandomRangeFloat(Minimum=1.4, Maximum=1.8), Use Linear Drag=true, Use Rotational Drag=true)
   &GravityForce(Gravity=&RandomRangeVector(Minimum=<0 0 -90>, Maximum=<0 0 -140>))
   &ScaleColor(
    Scale RGB=&VectorFromFloat(Value=&FloatFromCurve(FloatCurve=DataInterface() {propertyValues {Curve {keys [
          {interpMode=RCIM_Cubic, tangentMode=RCTM_Break, time=0.0952, value=3}
          {arriveTangent=-2.2104, interpMode=RCIM_Cubic, leaveTangent=-2.2104, time=0.2284, value=1.652}
          {interpMode=RCIM_Cubic, tangentMode=RCTM_Break, time=1, value=1}
         ]}}}))
    ScaleA=true
    ScaleRGB=true
   )
   &SolveForcesAndVelocity(Clamp Velocity=false, Limit Acceleration=false)
  ]
  rendererClasses [&NiagaraSpriteRendererProperties]
  renderers [{rendererClass=&NiagaraSpriteRendererProperties}]
 }
 Spark_02 {
  bEnabled=false
  emitterUpdateScript [
   &EmitterState()
   &SpawnBurst_Instantaneous(
    Spawn Count=&RandomRangeInt(Minimum=10, Maximum=15)
    Use Spawn Probability=false
    UseLoopCountLimit=false
   )
  ]
  particleSpawnScript [
   &InitializeParticle(
    Color=&SelectLinearColorFromArray(
     Array Sampling Mode=::NewEnumerator0
     Direct Array Index=?
     Color Selection Array=DataInterface() {propertyValues {ColorData [<rgba 0 0.1063 1 1>, <rgba 1 0.7667 0 1>, <rgba 0.6295 0 1 1>]}}
     Random Seed=0
    )
    Lifetime Max=0.8
    Lifetime Min=0.4
    Lifetime Mode=::NewEnumerator1
    Sprite Rotation Mode=::NewEnumerator2
    Sprite Size Mode=::NewEnumerator1
    Uniform Sprite Size Max=9
    Uniform Sprite Size Min=4
    UsePositionOffset=false
    Write Material Random=true
   )
   &ShapeLocation(MaskSpawnGroup=false, Override Seed=false, Sphere Radius=5, UseOwnerScale=true)
   &AddVelocity(
    Cone Angle=150
    Cone Axis=<0 0 1>
    Module.UseConeFalloff=false
    Module.UseVelDistribution=false
    Override Seed=false
    Velocity Mode=::NewEnumerator2
    Velocity Speed=&RandomRangeFloat(Minimum=90, Maximum=120)
   )
  ]
  particleUpdateScript [
   &ParticleState()
   &ScaleSpriteSize(Uniform Curve Sprite Scale=DataInterface() {propertyValues {Curve {keys [
        {interpMode=RCIM_Cubic, leaveTangent=15.2837, tangentMode=RCTM_User, time=0, value=0}
        {interpMode=RCIM_Cubic, tangentMode=RCTM_Break, time=0.1, value=1}
        {interpMode=RCIM_Cubic, tangentMode=RCTM_Break, time=1, value=0}
       ]}}})
   &SpriteRotationRate(Rotation Rate=&RandomRangeFloat(Minimum=-200, Maximum=200))
   &Drag(Drag=&RandomRangeFloat(Minimum=1.4, Maximum=1.8), Use Linear Drag=true, Use Rotational Drag=true)
   &GravityForce(Gravity=&RandomRangeVector(Minimum=<0 0 -90>, Maximum=<0 0 -140>))
   &ScaleColor(
    Scale RGB=&VectorFromFloat(Value=&FloatFromCurve(FloatCurve=DataInterface() {propertyValues {Curve {keys [
          {interpMode=RCIM_Cubic, tangentMode=RCTM_Break, time=0.0952, value=3}
          {arriveTangent=-2.2104, interpMode=RCIM_Cubic, leaveTangent=-2.2104, time=0.2284, value=1.652}
          {interpMode=RCIM_Cubic, tangentMode=RCTM_Break, time=1, value=1}
         ]}}}))
    ScaleA=true
    ScaleRGB=true
   )
   &SolveForcesAndVelocity(Clamp Velocity=false, Limit Acceleration=false)
  ]
  rendererClasses [&NiagaraSpriteRendererProperties]
  renderers [{rendererClass=&NiagaraSpriteRendererProperties}]
 }
}
summary {
 systemName=NS_Emote_LazyLizz_Spark
 emitters [
  {bEnabled=false, emitterName=Spark_01, rendererClasses [&NiagaraSpriteRendererProperties]}
  {bEnabled=false, emitterName=Spark_02, rendererClasses [&NiagaraSpriteRendererProperties]}
  {emitterName=Spark, rendererClasses [&NiagaraSpriteRendererProperties]}
 ]
 userVariables []
}
