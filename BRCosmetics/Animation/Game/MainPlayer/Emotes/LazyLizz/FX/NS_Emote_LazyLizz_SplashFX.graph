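# Niagara system for the LazyLizz emote that produces water splash and wake sprites on both sides of the emote.
# Splash_Both spawns at rates driven by left and right system pulse variables, then arcs droplets under gravity and drag.
# Wake_Both spawns per unit moved, drifting and widening sprites with time-of-day color curves.
# Each emitter covers both sides using spawn groups to mask mirrored offsets and velocities, with sub-UV frames and animated material parameters.
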NiagaraSystem /BRCosmetics/Animation/Game/MainPlayer/Emotes/LazyLizz/FX/NS_Emote_LazyLizz_SplashFX
dependencies {
 usedDataInterfaces [
  &NiagaraDataInterfaceSpriteRendererInfo
  &NiagaraDataInterfaceCurve
  &NiagaraDataInterfaceVector2DCurve
  &NiagaraDataInterfaceColorCurve
 ]
 usedDynamicInputs [&Lerp_Float]
 usedModules [
  &SystemState
  &self:SystemSpawnScript.SystemScriptSource.SystemScriptGraph.NiagaraNodeAssignment_0.SetVariables_#1
  &self:SystemSpawnScript.SystemScriptSource.SystemScriptGraph.NiagaraNodeAssignment_1.SetVariables_#2
 ]
}
emitters {
 Splash_Both {
  emitterUpdateScript [
   &EmitterState()
   &SpawnRate(SpawnRate=&Multiply_Float(A=1, B=$System.4Pulse_Left), Use Spawn Probability=false)
   &SpawnRate(SpawnGroup=1, SpawnRate=&Multiply_Float(A=1, B=$System.4Pulse_Right), Use Spawn Probability=false) {module=SpawnRate001}
  ]
  particleSpawnScript [
   &InitializeParticle(
    Color Maximum=<rgba 0.3843 0.6767 1 1>
    Color Minimum=<rgba 0.1548 0.5562 1 1>
    Color Mode=::NewEnumerator2
    Lifetime Max=0.55
    Lifetime Min=0.35
    Lifetime Mode=::NewEnumerator1
    Sprite Rotation Mode=::NewEnumerator2
    Sprite Size Mode=::NewEnumerator1
    Sprite UV Mode=::NewEnumerator3
    Uniform Sprite Size Max=30
    Uniform Sprite Size Min=18
    UsePositionOffset=false
    Write Material Random=true
   )
   &CameraOffset(Camera Offset Amount=3)
   &ShapeLocation(
    Box Size=<10 30 15>
    MaskSpawnGroup=true
    Offset=<24 30 7>
    Offset Mode=::NewEnumerator0
    Override Seed=false
    Rotation Mode=::NewEnumerator2
    Shape Primitive=::NewEnumerator2
    Surface Only=false
    UseOwnerScale=true
    Yaw / Pitch / Roll=<0.1 0 0>
   )
   &ShapeLocation(
    Box Size=<10 30 15>
    MaskSpawnGroup=true
    Offset=<-24 30 7>
    Offset Mode=::NewEnumerator0
    Override Seed=false
    Rotation Mode=::NewEnumerator2
    Shape Primitive=::NewEnumerator2
    Spawn Group Mask=1
    Surface Only=false
    UseOwnerScale=true
    Yaw / Pitch / Roll=<-0.1 0 0>
   ) {module=ShapeLocation001}
   &AddVelocity(Velocity=&MaskVector(
     Mask Source=0
     Mask Parameter=?
     Mask=0
     Masked Vector=&RandomRangeVector(Minimum=<30 -50 50>, Maximum=<50 -100 70>)
     Passthrough Vector=&MaskVector(
      Mask Source=0
      Mask Parameter=?
      Mask=1
      Masked Vector=&RandomRangeVector(Minimum=<-30 -50 50>, Maximum=<-50 -100 70>)
      Passthrough Vector=<1 1 1>
     )
    ))
   &SubUVAnimation(
    End Frame Range Override=3
    Evaluation Type=::NewEnumerator0
    "Sprite Renderer "=DataInterface() {propertyValues {SpriteRenderer=&self:Splash_Both_0.NiagaraSpriteRendererProperties_1}}
    SubUV Animation Mode:ENiagara_SubUVLookupModeV2=::NewEnumerator2
    UseEndFrame=true
    UseStartFrame=true
   )
  ]
  particleUpdateScript [
   &ParticleState()
   &SpriteRotationRate(Rotation Rate=&MaskFloat(
     Mask Source=0
     Mask Parameter=?
     Mask=0
     Masked Float=&RandomRangeFloat(Minimum=-70, Maximum=-100)
     Passthrough Float=&MaskFloat(
      Mask Source=0
      Mask Parameter=?
      Mask=1
      Masked Float=&RandomRangeFloat(Minimum=70, Maximum=100)
      Passthrough Float=1
     )
    ))
   &GravityForce(Gravity=&RandomRangeVector(Minimum=<0 0 -200>, Maximum=<0 0 -400>))
   &Drag(Drag=&RandomRangeFloat(Minimum=0.2, Maximum=0.3), Use Linear Drag=true, Use Rotational Drag=false)
   &SolveForcesAndVelocity(Clamp Velocity=false, Limit Acceleration=false)
   &ScaleSpriteSize(Uniform Curve Sprite Scale=DataInterface() {propertyValues {Curve {keys [
        {arriveTangent=4.1785, interpMode=RCIM_Cubic, leaveTangent=4.1785, tangentMode=RCTM_User, time=0, value=0.15}
        {arriveTangent=1.1179, interpMode=RCIM_Cubic, leaveTangent=1.1179, tangentMode=RCTM_User, time=0.2042, value=0.6228}
        {arriveTangent=0.0224, interpMode=RCIM_Cubic, leaveTangent=0.0224, tangentMode=RCTM_User, time=1, value=1}
       ]}}})
   &ScaleColor(
    Scale Alpha=&FloatFromCurve(
     FloatCurve=DataInterface() {propertyValues {Curve {keys [
         {arriveTangent=-0.2723, interpMode=RCIM_Cubic, leaveTangent=14.5791, tangentMode=RCTM_Break, time=0, value=0}
         {interpMode=RCIM_Cubic, tangentMode=RCTM_Break, time=0.1, value=1}
         {arriveTangent=0.0378, interpMode=RCIM_Cubic, leaveTangent=0.0378, tangentMode=RCTM_User, time=0.8, value=1}
         {arriveTangent=-3.7547, interpMode=RCIM_Cubic, leaveTangent=-3.7547, tangentMode=RCTM_Break, time=1, value=0}
        ]}}}
     Scale Curve=&FNTimeOfDayScalar(Night Scale=0.12, Morning Scale=0.12, Day Scale=0.12, Evening Scale=0.12)
    )
    Scale RGB=&VectorFromFloat(Value=5)
    ScaleA=true
    ScaleRGB=true
   )
   &DynamicMaterialParameters(
    Index 0 Param 1=&FloatFromCurve(FloatCurve=DataInterface() {propertyValues {Curve {keys [{interpMode=RCIM_Cubic, tangentMode=RCTM_Break, time=0.6, value=1}, {interpMode=RCIM_Cubic, tangentMode=RCTM_Break, time=1, value=0.4}]}}})
    Module.Index0Float0Write=true
    Module.Index0Float1Write=false
    Module.Index0Float2Write=false
    Module.Index0Float3Write=false
   )
  ]
  rendererClasses [&NiagaraSpriteRendererProperties]
  renderers [{rendererClass=&NiagaraSpriteRendererProperties}]
 }
 Wake_Both {
  emitterUpdateScript [
   &EmitterState()
   &SpawnPerUnit(
    Module.Use Max Movement Threshold=false
    Module.Use Movement Tolerance=true
    Movement Tolerance=4
    Spawn Spacing=35
    Use Spawn Probability=false
   )
   &SpawnPerUnit(
    Module.Use Max Movement Threshold=false
    Module.Use Movement Tolerance=true
    Movement Tolerance=4
    Spawn Group=1
    Spawn Spacing=35
    Use Spawn Probability=false
   ) {module=SpawnPerUnit001}
  ]
  particleSpawnScript [
   &InitializeParticle(
    Lifetime Max=0.9
    Lifetime Min=0.7
    Lifetime Mode=::NewEnumerator1
    Sprite Size Max=<40 20>
    Sprite Size Min=<30 15>
    Sprite Size Mode=::NewEnumerator2
    UsePositionOffset=false
    Write Material Random=true
   )
   &CameraOffset(Camera Offset Amount=3)
   &ShapeLocation(
    Box Size=<1 3 1>
    MaskSpawnGroup=true
    Offset=<27 40 5>
    Offset Mode=::NewEnumerator0
    Override Seed=false
    Rotation Mode=::NewEnumerator2
    Shape Primitive=::NewEnumerator2
    Surface Only=false
    UseOwnerScale=true
   ) {module=ShapeLocation001}
   &ShapeLocation(
    Box Size=<1 3 1>
    MaskSpawnGroup=true
    Offset=<-27 40 5>
    Offset Mode=::NewEnumerator0
    Override Seed=false
    Rotation Mode=::NewEnumerator2
    Shape Primitive=::NewEnumerator2
    Spawn Group Mask=1
    Surface Only=false
    UseOwnerScale=true
   )
   /SetVariables(Particles.SpriteFacing=&MaskVector(
     Mask Source=0
     Mask Parameter=?
     Mask=0
     Masked Vector=<-0.7 0 1>
     Passthrough Vector=&MaskVector(Mask Source=0, Mask Parameter=?, Mask=1, Masked Vector=<0.7 0 1>, Passthrough Vector=<1 1 1>)
    )) {isSetParameters=true}
   /SetVariables() {isSetParameters=true}
   &AddVelocity(Velocity=&MaskVector(
     Mask Source=0
     Mask Parameter=?
     Mask=0
     Masked Vector=&RandomRangeVector(Minimum=<5 -60 5>, Maximum=<15 -70 10>)
     Passthrough Vector=&MaskVector(
      Mask Source=0
      Mask Parameter=?
      Mask=1
      Masked Vector=&RandomRangeVector(Minimum=<-5 -60 5>, Maximum=<-15 -70 10>)
      Passthrough Vector=<1 1 1>
     )
    )) {module=AddVelocity001}
   &SubUVAnimation(
    End Frame Range Override=3
    Evaluation Type=::NewEnumerator0
    "Sprite Renderer "=DataInterface() {propertyValues {SpriteRenderer=&self:Wake_Both_1.NiagaraSpriteRendererProperties_1}}
    SubUV Animation Mode:ENiagara_SubUVLookupModeV2=::NewEnumerator2
    UseEndFrame=true
    UseStartFrame=true
   )
  ]
  particleUpdateScript [
   &ParticleState()
   &ScaleSpriteSize(
    Non-Uniform Curve Sprite Scale=DataInterface() {propertyValues {
      XCurve {keys [
        {interpMode=RCIM_Cubic, leaveTangent=1.0508, tangentMode=RCTM_Break, time=0, value=1}
        {interpMode=RCIM_Cubic, tangentMode=RCTM_Break, time=1, value=2}
       ]}
      YCurve {keys [
        {interpMode=RCIM_Cubic, leaveTangent=0.0002, tangentMode=RCTM_Break, time=0, value=1}
        {interpMode=RCIM_Cubic, tangentMode=RCTM_Break, time=1, value=1.4}
       ]}
     }}
    Scale Sprite Size Mode=::NewEnumerator3
   ) {module=ScaleSpriteSize002}
   &SolveForcesAndVelocity(Clamp Velocity=false, Limit Acceleration=false)
   &DynamicMaterialParameters(
    Index 0 Param 1=&FloatFromCurve(FloatCurve=DataInterface() {propertyValues {Curve {keys [{interpMode=RCIM_Cubic, tangentMode=RCTM_Break, time=0.4, value=1}, {interpMode=RCIM_Cubic, tangentMode=RCTM_Break, time=1, value=0.2}]}}})
    Module.Index0Float0Write=true
    Module.Index0Float1Write=false
    Module.Index0Float2Write=false
    Module.Index0Float3Write=false
   )
   &Color(
    Color=&FNTimeOfDayColor(
     Night Color=&ColorFromCurve(ColorCurve=DataInterface() {propertyValues {
        AlphaCurve {keys [<at 0 0>,<at 0.2 1 RCIM_Cubic>,<at 0.9 1>,<at 1 0 RCIM_Cubic>]}
        BlueCurve {keys [<at 0 0 RCIM_Cubic>,<at 0.5 1 RCIM_Cubic>,<at 0.8 0>]}
        GreenCurve {keys [<at 0 1 RCIM_Cubic>,<at 0.5 0.36 RCIM_Cubic>,<at 0.8 1>]}
        RedCurve {keys [<at 0 0 RCIM_Cubic>,<at 0.5 0.0114 RCIM_Cubic>,<at 0.8 0>]}
       }})
     Morning Color=&ColorFromCurve(ColorCurve=DataInterface() {propertyValues {
        AlphaCurve {keys [<at 0 0>,<at 0.2 1 RCIM_Cubic>,<at 0.9 1>,<at 1 0 RCIM_Cubic>]}
        BlueCurve {keys [<at 0 0 RCIM_Cubic>,<at 0.5 1 RCIM_Cubic>,<at 0.8 0>]}
        GreenCurve {keys [<at 0 1 RCIM_Cubic>,<at 0.5 0.36 RCIM_Cubic>,<at 0.8 1>]}
        RedCurve {keys [<at 0 0 RCIM_Cubic>,<at 0.5 0.0114 RCIM_Cubic>,<at 0.8 0>]}
       }})
     Day Color=&ColorFromCurve(ColorCurve=DataInterface() {propertyValues {
        AlphaCurve {keys [<at 0 0>,<at 0.2 1 RCIM_Cubic>,<at 0.9 1>,<at 1 0 RCIM_Cubic>]}
        BlueCurve {keys [<at 0 0.3 RCIM_Cubic>,<at 0.5 1 RCIM_Cubic>,<at 0.8 0.3>]}
        GreenCurve {keys [<at 0 1 RCIM_Cubic>,<at 0.5 0.552 RCIM_Cubic>,<at 0.8 1>]}
        RedCurve {keys [<at 0 0.3 RCIM_Cubic>,<at 0.5 0.308 RCIM_Cubic>,<at 0.8 0.3>]}
       }})
     Evening Color=&ColorFromCurve(ColorCurve=DataInterface() {propertyValues {
        AlphaCurve {keys [<at 0 0>,<at 0.2 1 RCIM_Cubic>,<at 0.9 1>,<at 1 0 RCIM_Cubic>]}
        BlueCurve {keys [<at 0 0 RCIM_Cubic>,<at 0.5 1 RCIM_Cubic>,<at 0.8 0>]}
        GreenCurve {keys [<at 0 1 RCIM_Cubic>,<at 0.5 0.36 RCIM_Cubic>,<at 0.8 1>]}
        RedCurve {keys [<at 0 0 RCIM_Cubic>,<at 0.5 0.0114 RCIM_Cubic>,<at 0.8 0>]}
       }})
    )
    Scale Alpha=0.6
    Scale Color=&VectorFromFloat(Value=1)
   ) {module=Color001}
  ]
  rendererClasses [&NiagaraSpriteRendererProperties]
  renderers [{rendererClass=&NiagaraSpriteRendererProperties}]
 }
}
summary {
 systemName=NS_Emote_LazyLizz_SplashFX
 emitters [
  {emitterName=Splash_Both, rendererClasses [&NiagaraSpriteRendererProperties]}
  {emitterName=Wake_Both, rendererClasses [&NiagaraSpriteRendererProperties]}
 ]
 userVariables [User.bIsInWater {type=NiagaraBool}]
}
