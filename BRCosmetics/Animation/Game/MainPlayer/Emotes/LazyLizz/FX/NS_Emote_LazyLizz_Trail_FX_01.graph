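# Niagara system for the LazyLizz emote that draws a single ribbon trail through one emitter, Main_Trail.
# Particles spawn at a steady rate with a constant velocity, mostly along negative Y with slight lift, and a curve tapers ribbon width over their life.
# A time-of-day color function keeps the trail white, with reduced alpha at night.
# Front-end and preview user booleans are exposed, but no shown module reads them.
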NiagaraSystem /BRCosmetics/Animation/Game/MainPlayer/Emotes/LazyLizz/FX/NS_Emote_LazyLizz_Trail_FX_01
dependencies {usedDataInterfaces [&NiagaraDataInterfaceCurve], usedDynamicInputs [], usedModules [&SystemState]}
emitters {Main_Trail {
  emitterUpdateScript [
   &EmitterState(Life Cycle Mode=::NewEnumerator1, Loop Behavior=::NewEnumerator1, Loop Duration Mode=::NewEnumerator1, UseLoopDelay=false)
   &SpawnRate(SpawnRate=50, Use Spawn Probability=false)
  ]
  particleSpawnScript [
   &InitializeParticle(
    Color=&FNTimeOfDayColor(Night Color=<rgba 1 1 1 0.4>, Morning Color=<rgba 1 1 1 1>, Day Color=<rgba 1 1 1 1>, Evening Color=<rgba 1 1 1 1>)
    Ribbon Facing Mode=::NewEnumerator1
    Ribbon Width=70
    Ribbon Width Mode=::NewEnumerator1
    UsePositionOffset=false
    Write Material Random=true
   )
   &AddVelocity(Velocity=<0 -200 19>)
  ]
  particleUpdateScript [
   &ParticleState()
   &SolveForcesAndVelocity(Clamp Velocity=false, Limit Acceleration=false)
   &ScaleRibbonWidth(Ribbon Width Scale=&FloatFromCurve(FloatCurve=DataInterface() {propertyValues {Curve {keys [{interpMode=RCIM_Cubic, tangentMode=RCTM_Break, time=0, value=1}, {interpMode=RCIM_Cubic, tangentMode=RCTM_Break, time=1, value=0.1}]}}}))
  ]
  rendererClasses [&NiagaraRibbonRendererProperties]
  renderers [{rendererClass=&NiagaraRibbonRendererProperties}]
 }}
summary {
 systemName=NS_Emote_LazyLizz_Trail_FX_01
 emitters [{emitterName=Main_Trail, rendererClasses [&NiagaraRibbonRendererProperties]}]
 userVariables [
  User.bIsFrontEnd {type=NiagaraBool}
  User.bIsFrontEndPreview {type=NiagaraBool}
 ]
}
